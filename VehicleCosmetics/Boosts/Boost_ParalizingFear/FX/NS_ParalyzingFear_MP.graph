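# Niagara system for the Paralyzing Fear vehicle boost cosmetic, drawing a glowing core with ribbon trails behind it.
# CoreYellow adds a glow sprite, while MainCore merges meshes and sprites into one emitter through per-spawn-group selectors, both offset from System.FinalPosition.
# Trails feeds five ribbon renderers using per-layer widths and colors, and TrailLightningBlack adds sine-displaced dark ribbons.
# Painted-variant overrides recolor most layers, and a frontend-preview flag changes trail velocity.
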NiagaraSystem /VehicleCosmetics/Boosts/Boost_ParalizingFear/FX/NS_ParalyzingFear_MP
dependencies {
 usedDataInterfaces [
  &NiagaraDataInterfaceCurve
  &NiagaraDataInterfaceVectorCurve
  &NiagaraDataInterfaceColorCurve
  &NiagaraDataInterfaceArrayFloat2
  &NiagaraDataInterfaceArrayFloat
  &NiagaraDataInterfaceArrayFloat3
  &NiagaraDataInterfaceSpriteRendererInfo
  &NiagaraDataInterfaceSkeletalMesh
 ]
 usedDynamicInputs []
 usedModules [
  &self:SystemSpawnScript.SystemScriptSource.SystemScriptGraph.NiagaraNodeAssignment_0.SetVariables_#1
  &SystemState
 ]
}
emitters {
 CoreYellow {
  emitterUpdateScript [
   &EmitterState(
    Inactive Response=::NewEnumerator1
    Life Cycle Mode=::NewEnumerator1
    Loop Behavior=::NewEnumerator1
    Loop Duration Mode=::NewEnumerator1
    UseLoopDelay=false
   )
   &SpawnBurst_Instantaneous(Use Spawn Probability=false, UseLoopCountLimit=false)
   &VC_FilteredSocketCount()
   &SpawnRate(SpawnRate=50, Use Spawn Probability=false) {enabled=false}
  ]
  particleSpawnScript [
   &VC_PaintedVariantsOverride(
    Black=<rgba 0.0104 0.0104 0.0104 0.913>
    BurntSienna=<rgba 0 0 0 0.2>
    Cobalt=<rgba 0 0.0277 1.5 1>
    Crimson=<rgba 0.5625 0 0.0152 0.6609>
    DefaultColor=<rgba 0.8177 0.741 0.6261 0.5043>
    ForestGreen=<rgba 0 0.5938 0.03 0.6522>
    Grey=<rgba 0.2865 0.2865 0.2865 0.713>
    Lime=<rgba 0.2995 0.599 0 0.6087>
    Orange=<rgba 0.5677 0.0747 0 0.7652>
    Pink=<rgba 0.4635 0.0555 0.3152 0.2>
    Purple=<rgba 0.0855 0 0.2083 0.8348>
    Saffron=<rgba 0.5625 0.4835 0.0996 0.8261>
    SkyBlue=<rgba 0 0.4436 0.6667 1>
    TitaniumWhite=<rgba 0.4792 0.4792 0.4792 1>
   )
   &InitializeParticle(
    Color=$StackContext.Color
    Position Offset=&AddVectorToPosition(Position=$System.FinalPosition, Vector=<-50 0 0>)
    Position Offset Coordinate Space=::Local
    Sprite Size Mode=::NewEnumerator3
    Uniform Sprite Size=15
    UsePositionOffset=true
    Write Material Random=true
   )
   &PartitionParticles(Number Of Partitions=$Emitter.PortCount) {enabled=false}
   &SkeletalMeshLocation(
    Bone / Socket Index=$Particles.PartitionParticles.Partition
    Bone / Socket Sampling Mode=::NewEnumerator5
    MaskSpawnGroup=false
    Mesh Sampling Type=::NewEnumerator6
    Skeletal Mesh=$System.SkeletalMesh
   ) {enabled=false}
   &AddVelocity(Velocity=<-1500 0 0>) {enabled=false}
   &CameraOffset(Camera Offset Amount=30)
  ]
  particleUpdateScript [
   &ParticleState(Kill Particles When Lifetime Has Elapsed=false)
   &ScaleSpriteSize(Scale Sprite Size Mode=::NewEnumerator0, Uniform Scale Factor=2.5)
   &SolveForcesAndVelocity(Clamp Velocity=false, Limit Acceleration=false)
  ]
  rendererClasses [&NiagaraSpriteRendererProperties]
  renderers [{rendererClass=&NiagaraSpriteRendererProperties}]
 }
 MainCore {
  emitterUpdateScript [
   &EmitterState(
    Inactive Response=::NewEnumerator1
    Life Cycle Mode=::NewEnumerator1
    Loop Delay=-1
    Loop Duration=1
    Scalability Mode=::NewEnumerator1
    UseLoopDelay=false
   )
   &VC_FilteredSocketCount()
   &SpawnBurst_Instantaneous(Spawn Count=4, Use Spawn Probability=false, UseLoopCountLimit=false)
   &SpawnBurst_Instantaneous(Spawn Count=2, Spawn Group=1, Use Spawn Probability=false, UseLoopCountLimit=false) {module=SpawnBurst_Instantaneous001}
   &SpawnBurst_Instantaneous(Spawn Count=2, Spawn Group=2, Use Spawn Probability=false, UseLoopCountLimit=false) {module=SpawnBurst_Instantaneous002}
   &SpawnRate(SpawnGroup=3, SpawnRate=4, Use Spawn Probability=false)
   &SpawnBurst_Instantaneous(Spawn Group=4, Use Spawn Probability=false, UseLoopCountLimit=false) {module=SpawnBurst_Instantaneous003}
  ]
  particleSpawnScript [
   &VC_PaintedVariantsOverride(
    DefaultColor=<rgba 0 0 0 1>
    R1 Black:LinearColor=<rgba 0.0312 0.0312 0.0312 1>
    R1 BurntSienna:LinearColor=<rgba 0.0938 0.0289 0.0116 1>
    R1 Cobalt:LinearColor=<rgba 0.0536 0.0394 0.2344 1>
    R1 Crimson:LinearColor=<rgba 0.1562 0.0293 0.0327 1>
    R1 DefaultColor=<rgba 0.32 0.1657 0.0088 1>
    R1 ForestGreen:LinearColor=<rgba 0.0157 0.151 0.0226 1>
    R1 Grey:LinearColor=<rgba 0.0885 0.0885 0.0885 1>
    R1 Lime:LinearColor=<rgba 0.099 0.14 0.0178 1>
    R1 Orange:LinearColor=<rgba 0.1146 0.0243 0.0048 1>
    R1 Pink:LinearColor=<rgba 0.38 0.0237 0.3153 1>
    R1 Purple:LinearColor=<rgba 0.0482 0 0.1046 1>
    R1 Saffron:LinearColor=<rgba 0.1198 0.1149 0.0455 1>
    R1 SkyBlue:LinearColor=<rgba 0 0.1733 0.2604 1>
    R1 TitaniumWhite:LinearColor=<rgba 0.151 0.151 0.151 1>
    R2 Black:LinearColor=<rgba 0.0348 0.0348 0.0348 1>
    R2 BurntSienna:LinearColor=<rgba 2 0.4231 0 1>
    R2 Cobalt:LinearColor=<rgba 0 0.4625 25 2>
    R2 Crimson:LinearColor=<rgba 25 1.3846 0.9115 1>
    R2 DefaultColor:LinearColor=<rgba 16.6366 0 25 1>
    R2 ForestGreen:LinearColor=<rgba 0 5 0.324 1>
    R2 Grey:LinearColor=<rgba 5 5 5 1>
    R2 Lime:LinearColor=<rgba 1.0435 2 0.087 1>
    R2 Orange:LinearColor=<rgba 5 0.3388 0 5>
    R2 Pink:LinearColor=<rgba 2 0 1.7128 85>
    R2 Purple:LinearColor=<rgba 0.5734 0 1.5 5>
    R2 Saffron:LinearColor=<rgba 1.5 1.0328 0 0.8>
    R2 SkyBlue:LinearColor=<rgba 0 0.9981 1.5 10>
    R2 TitaniumWhite:LinearColor=<rgba 0.4688 0.4688 0.4688 1>
    R3 Black:LinearColor=<rgba 0.0573 0.0573 0.0573 3>
    R3 BurntSienna:LinearColor=<rgba 0.6 0.0569 0.0002 3>
    R3 Cobalt:LinearColor=<rgba 0.2406 0.3139 4.2 2.5>
    R3 Crimson:LinearColor=<rgba 1.2 0.025 0.0568 5>
    R3 DefaultColor:LinearColor=<rgba 6 4.2 1.8 1>
    R3 ForestGreen:LinearColor=<rgba 0.0777 1.2 0 5>
    R3 Grey:LinearColor=<rgba 0.5 0.5 0.5 5>
    R3 Lime:LinearColor=<rgba 0.7293 1.35 0 5>
    R3 Orange:LinearColor=<rgba 1.2031 0.2115 0.0008 5>
    R3 Pink:LinearColor=<rgba 1.8 0.6094 1.367 5>
    R3 Purple:LinearColor=<rgba 0.5051 0.0541 1.1531 2.5>
    R3 Saffron:LinearColor=<rgba 2 1.5528 0.0833 5>
    R3 SkyBlue:LinearColor=<rgba 0.0879 1.9008 2.8125 5>
    R3 TitaniumWhite:LinearColor=<rgba 2 2 2 3>
    R4 Black:LinearColor=<rgba 0.0417 0.0417 0.0417 0.02>
    R4 BurntSienna:LinearColor=<rgba 0.154 0.0437 0.0152 0.02>
    R4 Cobalt:LinearColor=<rgba 0.1042 0.1572 0.3281 0.02>
    R4 Crimson:LinearColor=<rgba 0.2552 0.0266 0.0301 0.02>
    R4 DefaultColor:LinearColor=<rgba 0.2396 0.2396 0.2396 0.02>
    R4 ForestGreen:LinearColor=<rgba 0.05 0.3 0 0.02>
    R4 Grey:LinearColor=<rgba 0.16 0.16 0.16 0.02>
    R4 Lime:LinearColor=<rgba 0.251 0.35 0 0.02>
    R4 Orange:LinearColor=<rgba 0.349 0.0834 0.0145 0.02>
    R4 Pink:LinearColor=<rgba 0.4219 0.145 0.3212 0.02>
    R4 Purple:LinearColor=<rgba 0.1404 0.0304 0.2656 0.02>
    R4 Saffron:LinearColor=<rgba 0.3021 0.2597 0.0535 0.02>
    R4 SkyBlue:LinearColor=<rgba 0.0975 0.2336 0.3021 0.02>
    R4 TitaniumWhite:LinearColor=<rgba 0.2917 0.2917 0.2917 0.02>
    Use R1 Painted Overrides=true
    Use R2 Painted Overrides=true
    Use R3 Painted Overrides=true
    Use R4 Painted Overrides=true
   )
   /SetVariables(Particles.TagsSpheres:NiagaraInt32=&SpawnGroup()) {isSetParameters=true}
   &InitializeParticle(
    Color Mode=::NewEnumerator0
    Lifetime=/VehicleCosmetics/FX/Niagara/Modules/MergeEmitters/VC_FloatBySpawnGroup(
     SpawnGroup or ExecIndex:NiagaraInt32=0
     CustomInt:NiagaraInt32=0
     MaximumSpawnGroups:NiagaraInt32=5
     Float SG 0:NiagaraFloat=5
     Float SG 1:NiagaraFloat=5
     Float SG 2:NiagaraFloat=5
     Float SG 3:NiagaraFloat=0.3
     Float SG 4:NiagaraFloat=3
     Float SG 5:NiagaraFloat=0
    )
    Mesh Scale=/VehicleCosmetics/FX/Niagara/Modules/MergeEmitters/VC_VectorBySpawnGroup(
     SpawnGroup or ExecIndex:NiagaraInt32=0
     CustomInt:NiagaraInt32=0
     MaximumSpawnGroups:NiagaraInt32=3
     Vector Group 0:Vector3f=&Multiply_VectorByFloat(Vector=<0.9 0.9 0.1>, Float=4)
     Vector Group 1:Vector3f=&Multiply_VectorByFloat(Vector=<0.3 0.3 0.1>, Float=4.6)
     Vector Group 2:Vector3f=&Multiply_VectorByFloat(Vector=<0.5 0.5 0.1>, Float=5)
     Vector Group 3:Vector3f=<0 0 0>
     Vector Group 4:Vector3f=<0 0 0>
     Vector Group 5:Vector3f=<0 0 0>
    )
    Mesh Scale Mode=::NewEnumerator4
    Position Offset=&AddVectorToPosition(Position=$System.FinalPosition, Vector=<-50 0 0>)
    Sprite Rotation Angle=&RandomRangeFloat(
     Minimum=&SelectFloatFromArray(
      Direct Array Index=&SpawnGroup()
      Float Selection Array=DataInterface() {propertyValues {FloatData [0,0,0,-1,0]}}
     )
     Maximum=&RandomRangeFloat(
      Minimum=&SelectFloatFromArray(
       Direct Array Index=&SpawnGroup()
       Float Selection Array=DataInterface() {propertyValues {FloatData [0,0,0,1,0]}}
      )
      Maximum=0
     )
    )
    Sprite Rotation Mode=::NewEnumerator1
    Sprite Size Max=&SelectVector2DFromArray(
     Direct Array Index=&SpawnGroup()
     Vector 2D Selection Array=DataInterface() {propertyValues {FloatData [<0 0>, <0 0>, <0 0>, <30 36>, <30 30>]}}
    )
    Sprite Size Min=&SelectVector2DFromArray(
     Direct Array Index=&SpawnGroup()
     Vector 2D Selection Array=DataInterface() {propertyValues {FloatData [<0 0>, <0 0>, <0 0>, <15 23>, <30 30>]}}
    )
    Sprite Size Mode=::NewEnumerator2
    UsePositionOffset=true
    Write Material Random=true
   )
   &PartitionParticles(Number Of Partitions=$Emitter.PortCount) {enabled=false}
   &SkeletalMeshLocation(
    Bone / Socket Index=$Particles.PartitionParticles.Partition
    Bone / Socket Sampling Mode=::NewEnumerator5
    MaskSpawnGroup=false
    Mesh Sampling Type=::NewEnumerator6
    Skeletal Mesh=$System.SkeletalMesh
   ) {enabled=false}
   &AddVelocity(Velocity=<-1500 0 0>) {enabled=false}
   &Color(
    Color=<rgba 0 0 0 1>
    Scale Alpha=&VC_FloatBySpawnGroup(
     SpawnGroup or ExecIndex:NiagaraInt32=0
     CustomInt:NiagaraInt32=0
     MaximumSpawnGroups:NiagaraInt32=5
     Float SG 0:NiagaraFloat=90
     Float SG 1:NiagaraFloat=100
     Float SG 2:NiagaraFloat=50
     Float SG 3:NiagaraFloat=1
     Float SG 4:NiagaraFloat=1
     Float SG 5:NiagaraFloat=0
    )
    Scale Color=&VC_VectorBySpawnGroup(
     SpawnGroup or ExecIndex:NiagaraInt32=0
     CustomInt:NiagaraInt32=0
     MaximumSpawnGroups:NiagaraInt32=5
     Vector Group 0:Vector3f=<1 1 1>
     Vector Group 1:Vector3f=<1 1 1>
     Vector Group 2:Vector3f=<1 1 1>
     Vector Group 3:Vector3f=<1 1 1>
     Vector Group 4:Vector3f=<1 1 1>
     Vector Group 5:Vector3f=<0 0 0>
    )
   )
   &CameraOffset(Camera Offset Amount=&VC_FloatBySpawnGroup(
     SpawnGroup or ExecIndex:NiagaraInt32=0
     CustomInt:NiagaraInt32=0
     MaximumSpawnGroups:NiagaraInt32=5
     Float SG 0:NiagaraFloat=-30
     Float SG 1:NiagaraFloat=65
     Float SG 2:NiagaraFloat=15
     Float SG 3:NiagaraFloat=&RandomRangeFloat(Minimum=55, Maximum=55)
     Float SG 4:NiagaraFloat=0
     Float SG 5:NiagaraFloat=0
    ))
   &InitialMeshOrientation(
    Mesh Orientation Mode=::NewEnumerator7
    Rotation=&SelectVectorFromArray(
     Direct Array Index=&ModuloInt(A=&ReturnParticleIDAttributes(Particle ID Attribute=::NewEnumerator1), B=3)
     Vector Selection Array=DataInterface() {propertyValues {FloatData [<0 0 0>, <0 0.25 0>, <-0.25 0 0>]}}
    )
    Use Rotation Vector=true
   )
   /SetVariables(Particles.WidthRaysCore:NiagaraFloat=/Niagara/DynamicInputs/TypeConversions/Vector2DLength(Vector:Vector2f=<0 0>)) {isSetParameters=true}
  ]
  particleUpdateScript [
   &ParticleState()
   &UpdateMeshOrientation(Rotation Vector=&VC_VectorBySpawnGroup(
     SpawnGroup or ExecIndex:NiagaraInt32=0
     CustomInt:NiagaraInt32=0
     MaximumSpawnGroups:NiagaraInt32=3
     Vector Group 0:Vector3f=&RandomRangeVector(Minimum=<0 0 0>, Maximum=<0.2 0.2 0.2>)
     Vector Group 1:Vector3f=&RandomRangeVector(Minimum=<0 0 0>, Maximum=<0.2 0.2 0.2>)
     Vector Group 2:Vector3f=&RandomRangeVector(Minimum=<0 0 0>, Maximum=<0.2 0.2 0.2>)
     Vector Group 3:Vector3f=<0 0 0>
     Vector Group 4:Vector3f=<0 0 0>
     Vector Group 5:Vector3f=<0 0 0>
    ))
   &SubUVAnimation(
    "Sprite Renderer "=DataInterface() {propertyValues {SpriteRenderer=&self:MainCore_0.NiagaraSpriteRendererProperties_5}}
    UseEndFrame=false
    UseStartFrame=false
   )
   &ScaleColor(Scale Alpha=0.01, Scale RGB=<8 8 8>, ScaleA=true, ScaleRGB=true)
   &ScaleSpriteSize(Scale Sprite Size Mode=::NewEnumerator0, Uniform Scale Factor=3)
   &ScaleMeshSize(Scale Factor=&VC_VectorBySpawnGroup(
     SpawnGroup or ExecIndex:NiagaraInt32=0
     CustomInt:NiagaraInt32=0
     MaximumSpawnGroups:NiagaraInt32=3
     Vector Group 0:Vector3f=&VectorFromCurve(VectorCurve=DataInterface() {propertyValues {XCurve {keys [<at 0 1>,<at 1 1>]}, YCurve {keys [<at 0 1>,<at 1 1>]}, ZCurve {keys [<at 0 0>,<at 1 0.5>]}}})
     Vector Group 1:Vector3f=&VectorFromCurve(VectorCurve=DataInterface() {propertyValues {XCurve {keys [<at 0 1.2>,<at 1 1>]}, YCurve {keys [<at 0 1.2>,<at 1 1>]}, ZCurve {keys [<at 0 0>,<at 1 0.5>]}}})
     Vector Group 2:Vector3f=&VectorFromCurve(
      VectorCurve=DataInterface() {propertyValues {XCurve {keys [<at 0 1.1>,<at 1 1.01>]}, YCurve {keys [<at 0 1.1>,<at 1 1.01>]}, ZCurve {keys [<at 0 0>,<at 1 0.5>]}}}
      Scale Curve=<0.8 0.8 0.8>
     )
     Vector Group 3:Vector3f=<0 0 0>
     Vector Group 4:Vector3f=<0 0 0>
     Vector Group 5:Vector3f=<0 0 0>
    ))
   &DynamicMaterialParameters(
    Index 0 Param 1=&RandomRangeFloat(Minimum=2, Maximum=4)
    Index 0 Param 2=&RandomRangeFloat(Minimum=1, Maximum=2)
    Module.Index0Param0Write=true
    Module.Index0Param1Write=false
    Module.Index0Param2Write=true
    Module.Index0Param3Write=true
   )
   &SolveForcesAndVelocity(Clamp Velocity=false, Limit Acceleration=false)
   &JitterPosition(Jitter Amount:NiagaraFloat=&VC_FloatBySpawnGroup(
     SpawnGroup or ExecIndex:NiagaraInt32=0
     CustomInt:NiagaraInt32=0
     MaximumSpawnGroups:NiagaraInt32=5
     Float SG 0:NiagaraFloat=0
     Float SG 1:NiagaraFloat=0
     Float SG 2:NiagaraFloat=0
     Float SG 3:NiagaraFloat=0
     Float SG 4:NiagaraFloat=1
     Float SG 5:NiagaraFloat=0
    ))
  ]
  rendererClasses [&NiagaraMeshRendererProperties, &NiagaraSpriteRendererProperties]
  renderers [
   {rendererClass=&NiagaraMeshRendererProperties, rendererIndex=0}
   {rendererClass=&NiagaraMeshRendererProperties, rendererIndex=1}
   {rendererClass=&NiagaraMeshRendererProperties, rendererIndex=2}
   {rendererClass=&NiagaraSpriteRendererProperties, rendererIndex=3}
   {rendererClass=&NiagaraSpriteRendererProperties, rendererIndex=4}
  ]
 }
 TrailLightningBlack {
  emitterUpdateScript [
   &EmitterState()
   &VC_FilteredSocketCount()
   &VC_RateQualityHandler(High Spawn Rate=60, Low Spawn Rate=30, Medium Spawn Rate=40, Use Custom Rate?=true)
   &VC_SpawnRibbon_v1(Number of Ribbons=$Emitter.PortCount, SpawnRate=$Emitter.Spawn Rate)
  ]
  particleSpawnScript [
   &VC_PaintedVariantsOverride() {enabled=false}
   &VC_InitializeRibbon_v1(
    Custom Offset=true
    Lifetime=0.3
    Local Vector Offset=&AddVectorToPosition(Position=$System.FinalPosition, Vector=<-40 0 0>)
    Offset Amount=1
    Ribbon Width=6
   )
   &AddVelocity(Velocity=&MakeCustomVectorFromBool(Bool=$User.bIsFrontendPreview, True Vector=<-1500 0 1>, False Vector=<-1000 0 0>)) {module=AddVelocity001}
   &Color(Color=<rgba 0 0 0 1>)
  ]
  particleUpdateScript [
   &ParticleState()
   /SetVariables(Particles.RibbonTwist:NiagaraFloat=&FloatFromCurve(FloatCurve=DataInterface() {propertyValues {Curve {keys [<at 0 0>,<at 1 1>]}}})) {isSetParameters=true}
   &ScaleColor(
    Scale Alpha=&FloatFromCurve(FloatCurve=DataInterface() {propertyValues {Curve {keys [<at 0 0>,<at 0.1 1>,<at 0.9 1>,<at 1 0>]}}}, Scale Curve=0.45)
    Scale RGB=<3 3 3>
    ScaleA=true
    ScaleRGB=true
   )
   &ScaleRibbonWidth(Ribbon Width Scale=&MakeFloatFromVector(
     Vector=&VectorFromCurve(VectorCurve=DataInterface() {propertyValues {XCurve {keys [<at 0 0.5>,<at 0.2 1.2>,<at 1 0.8>]}, YCurve {keys [<at 0 0>,<at 1 0>]}, ZCurve {keys [<at 0 0>,<at 1 0>]}}})
     Channel=::NewEnumerator0
    ))
   &InheritSourceMovement(
    Applied Position Scale=&VectorFromFloat(Value=&FloatFromCurve(FloatCurve=DataInterface() {propertyValues {Curve {keys [
          {interpMode=RCIM_Cubic, leaveTangent=-0.0539, tangentMode=RCTM_Break, time=0, value=1}
          {arriveTangent=-1.8329, interpMode=RCIM_Cubic, tangentMode=RCTM_Break, time=1, value=0.5}
         ]}}}))
    ApplyForce=false
    ApplyPosition=true
    ApplyPositionFalloff=false
   )
   &ScaleRibbonWidth(Ribbon Width Scale=2.5) {module=ScaleRibbonWidth001}
   &SolveForcesAndVelocity(Clamp Velocity=false, Limit Acceleration=false)
   &VC_UpdateRibbon_v1()
   /SetVariables(Particles.Position=&AddVectorToPosition(
     Position=$Particles.Position
     Vector=&TransformVector(
      Vector=&MakeVector(
       X=0
       Y=&Sine(
        Normalized Angle=&Add_Float(
         A=&Multiply_Float(A=$Particles.NormalizedAge, B=-2)
         B=&Multiply_Float(A=$Emitter.LoopedAge, B=1)
        )
        Period=1
        Scale=&FloatFromCurve(FloatCurve=DataInterface() {propertyValues {Curve {keys [<at 0.15 0>,<at 0.8 20>]}}}, Scale Curve=2.6)
        Bias=0
       )
       Z=&Sine(
        Normalized Angle=&Add_Float(
         A=&Multiply_Float(A=$Particles.NormalizedAge, B=-2)
         B=&Multiply_Float(A=$Emitter.LoopedAge, B=1)
        )
        Period=1
        Scale=&FloatFromCurve(FloatCurve=DataInterface() {propertyValues {Curve {keys [<at 0.15 0>,<at 0.8 20>]}}})
        Bias=0
       )
      )
      Destination Space=::Simulation
      Apply Scale=true
     )
    )) {isSetParameters=true}
  ]
  rendererClasses [&NiagaraRibbonRendererProperties]
  renderers [{rendererClass=&NiagaraRibbonRendererProperties}]
 }
 Trails {
  emitterUpdateScript [
   &EmitterState()
   &VC_FilteredSocketCount()
   &VC_RateQualityHandler(Epic Spawn Rate=30, High Spawn Rate=30, Low Spawn Rate=20, Medium Spawn Rate=25, Use Custom Rate?=true)
   &VC_SpawnRibbon_v1(Number of Ribbons=$Emitter.PortCount, Spawn Time=-1, SpawnRate=$Emitter.Spawn Rate)
  ]
  particleSpawnScript [
   &VC_PaintedVariantsOverride(
    Black=<rgba 0.058 0.058 0.058 0.8>
    BurntSienna=<rgba 0.1979 0.0611 0.0244 0.9>
    Cobalt=<rgba 0.1006 0.1086 0.5365 0.9>
    Crimson=<rgba 0.4115 0 0.0112 0.6>
    DefaultColor=<rgba 1.02 0.2277 0 1>
    ForestGreen=<rgba 0 0.4115 0.0208 0.85>
    Grey=<rgba 0.2135 0.2135 0.2135 0.9>
    Lime=<rgba 0.2838 0.4062 0.0523 0.9>
    Orange=<rgba 0.3958 0.0582 0 0.8>
    Pink=<rgba 0.6094 0.1047 0.4259 0.8>
    Purple=<rgba 0.1352 0.0537 0.2292 0.8>
    R1 BurntSienna:LinearColor=<rgba 0.1042 0.022 0 1>
    R1 Cobalt:LinearColor=<rgba 0 0.0956 1.1063 1>
    R1 Crimson:LinearColor=<rgba 1 0.099 0.1234 1>
    R1 DefaultColor=<rgba 0.464 0 0.7 1>
    R1 ForestGreen:LinearColor=<rgba 0.1283 0.8958 0.1101 1>
    R1 Grey:LinearColor=<rgba 0.4271 0.4271 0.4271 1>
    R1 Lime:LinearColor=<rgba 0.5316 0.7917 0.0948 1>
    R1 Orange:LinearColor=<rgba 0.8177 0.1666 0.0511 1>
    R1 Pink:LinearColor=<rgba 0.9062 0.1794 0.751 1>
    R1 Purple:LinearColor=<rgba 0.3146 0.0265 0.6354 1>
    R1 Saffron:LinearColor=<rgba 0.6094 0.5271 0.1904 1>
    R1 SkyBlue:LinearColor=<rgba 0.1274 0.6717 0.875 1>
    R1 TitaniumWhite:LinearColor=<rgba 0.6302 0.6302 0.6302 1>
    R2 Black:LinearColor=<rgba 0.151 0.151 0.151 1>
    R2 BurntSienna:LinearColor=<rgba 0.7396 0.1565 0 1>
    R2 Cobalt:LinearColor=<rgba 0.0911 0.1357 2.5 2>
    R2 Crimson:LinearColor=<rgba 1.2 0 0.0325 2>
    R2 DefaultColor:LinearColor=<rgba 3 3 3 5>
    R2 ForestGreen:LinearColor=<rgba 0 1 0.0505 2>
    R2 Grey:LinearColor=<rgba 0.6562 0.6562 0.6562 2>
    R2 Lime:LinearColor=<rgba 0.8057 1.25 0 2>
    R2 Orange:LinearColor=<rgba 1.02 0.3558 0 2>
    R2 Pink:LinearColor=<rgba 2.2 0.5615 1.6042 2>
    R2 Purple:LinearColor=<rgba 0.513 0 1.25 2>
    R2 Saffron:LinearColor=<rgba 2 1.5333 0 2>
    R2 SkyBlue:LinearColor=<rgba 0 1.1977 1.8 2>
    R2 TitaniumWhite:LinearColor=<rgba 1 1 1 2>
    R3 SkyBlue:LinearColor=<rgba 0 0.357 0.5365 1>
    Saffron=<rgba 0.3 0.3 0.041 0.9>
    SkyBlue=<rgba 0 0.4644 0.6979 0.9>
    TitaniumWhite=<rgba 0.3542 0.3542 0.3542 0.9>
    Use R1 Painted Overrides=true
    Use R2 Painted Overrides=true
    Use R3 Painted Overrides=true
   )
   &VC_InitializeRibbon_v1(
    Color=$StackContext.Color
    Custom Offset=true
    Lifetime=0.8
    Local Vector Offset=&AddVectorToPosition(Position=$System.FinalPosition, Vector=<-35 0 0>)
    Offset Amount=1
    Ribbon Width=15
   )
   &AddVelocity(Velocity=&MakeCustomVectorFromBool(Bool=$User.bIsFrontendPreview, True Vector=<-1500 0 0>, False Vector=<-1000 0 0>)) {module=AddVelocity001}
   /SetVariables(Particles.RibbonWidthBackTrail=40, Particles.RibbonWidthCoreTrail=30, Particles.RibbonWidthLightning=80, Particles.RibbonWidthPurpleTrail=90) {isSetParameters=true}
  ]
  particleUpdateScript [
   &ParticleState()
   &ScaleColor(
    Scale Alpha=&FloatFromCurve(FloatCurve=DataInterface() {propertyValues {Curve {keys [
         <at 0 0>
         <at 0.1 1>
         {interpMode=RCIM_Cubic, leaveTangent=-5.6055, tangentMode=RCTM_Break, time=0.4, value=1}
         {arriveTangent=-0.4645, interpMode=RCIM_Cubic, tangentMode=RCTM_Break, time=1, value=0}
        ]}}})
    Scale RGB=<2 2 2>
    ScaleA=true
    ScaleRGB=true
   )
   &ScaleRibbonWidth(Ribbon Width Scale=&MakeFloatFromVector(
     Vector=&VectorFromCurve(VectorCurve=DataInterface() {propertyValues {XCurve {keys [<at 0 1>,<at 0.05 1.2>,<at 0.1 1.4>,<at 1 0.8>]}, YCurve {keys [<at 0 0>,<at 1 0>]}, ZCurve {keys [<at 0 0>,<at 1 0>]}}})
     Channel=::NewEnumerator0
    ))
   &InheritSourceMovement(
    Applied Position Scale=&VectorFromFloat(Value=&FloatFromCurve(FloatCurve=DataInterface() {propertyValues {Curve {keys [
          {interpMode=RCIM_Cubic, tangentMode=RCTM_Break, time=0, value=1}
          {arriveTangent=-2.9025, interpMode=RCIM_Cubic, tangentMode=RCTM_Break, time=1, value=0}
         ]}}}))
    ApplyForce=false
    ApplyPosition=true
    ApplyPositionFalloff=false
   )
   &SolveForcesAndVelocity(Clamp Velocity=false, Limit Acceleration=false)
   &VC_UpdateRibbon_v1(Enable Skeletal Position=true)
   /SetVariables(
    Particles.BackTrail:LinearColor=&MultiplyLinearColor(
     Linear Color=<rgba 0 0 0 1>
     Scale Factor=&Vector4FromCurve(Vector4Curve=DataInterface() {propertyValues {
        AlphaCurve {keys [
          <at 0 0>
          <at 0.1 1.5>
          {interpMode=RCIM_Cubic, leaveTangent=-8.1388, tangentMode=RCTM_Break, time=0.4, value=1.5}
          {arriveTangent=-0.904, interpMode=RCIM_Cubic, leaveTangent=-0.7099, tangentMode=RCTM_Break, time=1, value=0}
         ]}
        BlueCurve {keys [<at 0 3>,<at 1 3>]}
        GreenCurve {keys [<at 0 3>,<at 1 3>]}
        RedCurve {keys [<at 0 3>,<at 1 3>]}
       }})
    )
    Particles.LightningColor:LinearColor=&MakeCustomLinearColorFromBool(
     Bool=$User.bIsPainted
     True Color=&MultiplyLinearColor(
      Linear Color=$User.PaintedColor
      Scale Factor=&Vector4FromCurve(Vector4Curve=DataInterface() {propertyValues {
         AlphaCurve {keys [
           {interpMode=RCIM_Cubic, tangentMode=RCTM_Break, time=0, value=0}
           {arriveTangent=-3.2174, interpMode=RCIM_Cubic, leaveTangent=-7.4246, tangentMode=RCTM_Break, time=0.1597, value=0.8}
           <at 0.3712 0.05>
           {interpMode=RCIM_Cubic, tangentMode=RCTM_Break, time=0.6419, value=0}
          ]}
         BlueCurve {keys [<at 0 12>,<at 1 1.5>]}
         GreenCurve {keys [<at 0 12>,<at 1 1.5>]}
         RedCurve {keys [<at 0 12>,<at 1 1.5>]}
        }})
     )
     False Color=&MultiplyLinearColor(
      Linear Color=$User.PaintedColor
      Scale Factor=&Vector4FromCurve(Vector4Curve=DataInterface() {propertyValues {
         AlphaCurve {keys [
           {interpMode=RCIM_Cubic, tangentMode=RCTM_Break, time=0, value=0}
           {arriveTangent=-3.2174, interpMode=RCIM_Cubic, leaveTangent=-7.4246, tangentMode=RCTM_Break, time=0.1597, value=1}
           <at 0.2913 0.05>
           {interpMode=RCIM_Cubic, tangentMode=RCTM_Break, time=0.6419, value=0}
          ]}
         BlueCurve {keys [<at 0 12>,<at 1 1.5>]}
         GreenCurve {keys [<at 0 12>,<at 1 1.5>]}
         RedCurve {keys [<at 0 12>,<at 1 1.5>]}
        }})
     )
    )
    Particles.PurpleTrail:LinearColor=&MakeCustomLinearColorFromBool(
     Bool=$User.bIsPainted
     True Color=&MultiplyLinearColor(
      Linear Color=$Particles.Initial.R1 Color
      Scale Factor=&Vector4FromCurve(Vector4Curve=DataInterface() {propertyValues {
         AlphaCurve {keys [
           <at 0 0>
           <at 0.1 1.3>
           {interpMode=RCIM_Cubic, leaveTangent=-4.3322, tangentMode=RCTM_Break, time=0.3328, value=1.3}
           {arriveTangent=-0.8785, interpMode=RCIM_Cubic, tangentMode=RCTM_Break, time=1, value=0}
          ]}
         BlueCurve {keys [<at 0 1>,<at 1 1>]}
         GreenCurve {keys [<at 0 1>,<at 1 1>]}
         RedCurve {keys [<at 0 1>,<at 1 1>]}
        }})
     )
     False Color=&MultiplyLinearColor(
      Linear Color=$Particles.Initial.R1 Color
      Scale Factor=&Vector4FromCurve(Vector4Curve=DataInterface() {propertyValues {
         AlphaCurve {keys [<at 0.9063 1>,<at 1 0>]}
         BlueCurve {keys [<at 0 1>,<at 1 1>]}
         GreenCurve {keys [<at 0 1>,<at 1 1>]}
         RedCurve {keys [<at 0 1>,<at 1 1>]}
        }})
     )
    )
   ) {isSetParameters=true}
   /SetVariables(Particles.DMPLightning:Vector4f=&MakeVector4(X=&RandomRangeFloat(Minimum=0, Maximum=1), Y=0, Z=0, W=0)) {isSetParameters=true}
  ]
  rendererClasses [&NiagaraRibbonRendererProperties]
  renderers [
   {rendererClass=&NiagaraRibbonRendererProperties}
   {rendererClass=&NiagaraRibbonRendererProperties, rendererIndex=1}
   {rendererClass=&NiagaraRibbonRendererProperties, rendererIndex=2}
   {rendererClass=&NiagaraRibbonRendererProperties, rendererIndex=3}
   {rendererClass=&NiagaraRibbonRendererProperties, rendererIndex=4}
  ]
 }
}
summary {
 systemName=NS_ParalyzingFear_MP
 emitters [
  {emitterName=Trails, rendererClasses [&NiagaraRibbonRendererProperties]}
  {emitterName=MainCore, rendererClasses [&NiagaraMeshRendererProperties, &NiagaraSpriteRendererProperties]}
  {emitterName=CoreYellow, rendererClasses [&NiagaraSpriteRendererProperties]}
  {emitterName=TrailLightningBlack, rendererClasses [&NiagaraRibbonRendererProperties]}
 ]
 userVariables [
  User.AuraAlpha {type=NiagaraFloat, defaultValue=1.5, description=""}
  User.BackColor {type=LinearColor, description="", defaultValue {struct=LinearColor, value=<rgba 0.05 0 0.1 1>}}
  User.bIsFrontendPreview {type=NiagaraBool}
  User.bIsPainted {type=NiagaraBool}
  User.CenterBrightness {type=NiagaraFloat, defaultValue=55, description=""}
  User.ColorVariantID {type=NiagaraInt32}
  User.DTCarOffset {type=Vector3f, description="", defaultValue {struct=Vector3f, value=<0 0 0>}}
  User.GameModeID {type=NiagaraInt32}
  User.OrangeVoidCone {type=LinearColor, description="", defaultValue {struct=LinearColor, value=<rgba 1 0.15 0 1>}}
  User.PaintedColor {type=LinearColor, description="", defaultValue {struct=LinearColor, value=<rgba 1 1 1 1>}}
  User.PaintedScalar {type=NiagaraFloat, defaultValue=1, description=""}
  User.PortCount {type=NiagaraInt32, defaultValue=1}
  User.SingleBoostPos {type=NiagaraPosition, description="", defaultValue {}}
 ]
}
